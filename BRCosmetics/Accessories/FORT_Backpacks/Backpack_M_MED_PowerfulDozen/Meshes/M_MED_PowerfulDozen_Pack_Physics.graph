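# Physics asset for the PowerfulDozen backpack cosmetic, defining collision bodies and joint limits for its pack, two knives and dangling straps.
# Each side has a three-capsule strap chain (A, B, C) hanging from a tiny MSTR anchor body, with almost no twist allowed.
# Both long knife capsules constrain to the large pack capsule with equal swing and twist limits.
# No constraint links the strap anchors to the pack.
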PhysicsAsset /BRCosmetics/Accessories/FORT_Backpacks/Backpack_M_MED_PowerfulDozen/Meshes/M_MED_PowerfulDozen_Pack_Physics
bodies {
 dyn_left_strap_A [{shapeType=Capsule, length=4.1, radius=0.8, shapeName=dyn_strap_01_L_capsule, center=<2.54 0 0>, rotation=<rot 90 -0.0293 -0.0297>}]
 dyn_left_strap_B [{shapeType=Capsule, length=9, radius=1.5, shapeName=dyn_strap_02_L_capsule, center=<0.7336 0 0>, rotation=<rot 90 -0.1581 -0.1583>}]
 dyn_left_strap_C [{shapeType=Capsule, length=6, radius=1.5, shapeName=dyn_strap_03_L_capsule, center=<3.2918 0 0>, rotation=<rot 90 -0.1581 -0.1583>}]
 dyn_left_strap_MSTR [{shapeType=Capsule, length=0, radius=0.5, shapeName=strap_MSTR_L_capsule, rotation=<rot 90 -90.0001 -90.0005>}]
 dyn_right_strap_A [{shapeType=Capsule, length=4.1, radius=0.95, shapeName=dyn_strap_01_R_capsule, center=<2.891 0 0>, rotation=<rot 90 -6.7215 -6.7219>}]
 dyn_right_strap_B [{shapeType=Capsule, length=7, radius=1.5, shapeName=dyn_strap_02_R_capsule, center=<0.8558 0 0>, rotation=<rot 90 -90 -90.0002>}]
 dyn_right_strap_C [{shapeType=Capsule, length=6, radius=1.5, shapeName=dyn_strap_03_R_capsule, center=<1.8907 0 0>, rotation=<rot 90 -160.5288 -160.529>}]
 dyn_right_strap_MSTR [{shapeType=Capsule, length=0, radius=0.5, shapeName=strap_MSTR_R_capsule, rotation=<rot 90 -90 -90.0004>}]
 left_knife [{shapeType=Capsule, length=47, radius=1.6, center=<8.0492 0.2515 -0.3197>, rotation=<rot 0 -89.3751 -89.3117>}]
 pack [{shapeType=Capsule, length=14, radius=13.9951, center=<0.7044 7.9597 -4.2013>, rotation=<rot 0 -90.7921 -89.9848>}]
 right_knife [{shapeType=Capsule, length=47, radius=1.6, center=<8.0492 0.2515 -0.3197>, rotation=<rot 0 -89.3751 -89.3117>}]
}
constraints [
 {bone1Name=right_knife, bone2Name=pack, swing1LimitDegrees=45, swing2LimitDegrees=45, twistLimitDegrees=45}
 {bone1Name=left_knife, bone2Name=pack, swing1LimitDegrees=45, swing2LimitDegrees=45, twistLimitDegrees=45}
 {bone1Name=dyn_right_strap_A, bone2Name=dyn_right_strap_MSTR, swing1LimitDegrees=55, swing2LimitDegrees=20, twistLimitDegrees=0}
 {bone1Name=dyn_left_strap_C, bone2Name=dyn_left_strap_B, swing1LimitDegrees=45, swing2LimitDegrees=45, twistLimitDegrees=2}
 {bone1Name=dyn_left_strap_A, bone2Name=dyn_left_strap_MSTR, swing1LimitDegrees=55, swing2LimitDegrees=20, twistLimitDegrees=1}
 {bone1Name=dyn_left_strap_B, bone2Name=dyn_left_strap_A, swing1LimitDegrees=45, swing2LimitDegrees=45, twistLimitDegrees=2}
 {bone1Name=dyn_right_strap_C, bone2Name=dyn_right_strap_B, swing1LimitDegrees=45, swing2LimitDegrees=45, twistLimitDegrees=2}
 {bone1Name=dyn_right_strap_B, bone2Name=dyn_right_strap_A, swing1LimitDegrees=45, swing2LimitDegrees=45, twistLimitDegrees=2}
]
