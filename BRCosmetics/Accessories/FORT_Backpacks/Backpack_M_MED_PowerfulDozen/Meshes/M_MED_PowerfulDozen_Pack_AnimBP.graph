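# Animation Blueprint for the PowerfulDozen backpack, handling strap physics, pack fitting and visibility of one material element.
# It additively offsets the pack bone from spine bend except on three name-matched body meshes, then caches the pose.
# A bool selects RigidBody or chained AnimDynamics straps with knife-bone planar limits.
# Events toggle "Hide Element 01", hiding it during emotes or while the tagged pickaxe is equipped outside skydiving.
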AnimBlueprint /BRCosmetics/Accessories/FORT_Backpacks/Backpack_M_MED_PowerfulDozen/Meshes/M_MED_PowerfulDozen_Pack_AnimBP
target_skeleton=&/M_MED_PowerfulDozen_Pack_Skeleton
graphs [
 AnimGraph {nodes [
   n1 = Anim Dynamics - Bone: dyn_left_strap_A!(
    ComponentPose=Local To Component(LocalPose=Use cached pose 'INPUT POSE'())
    Alpha=1
    GravityOverride:Vector="0,0,-980"
    ExternalForce="0,0,0"
   ) -> (Pose) {
    angularBiasOverride=1
    angularDampingOverride=1
    angularSpringConstant=3000
    bAngularSpring=true
    bOverrideAngularBias=true
    bOverrideAngularDamping=true
    bUseGravityOverride=true
    simulationSpace=RootRelative
    alphaBoolBlend {blendOption=Linear}
    boundBone {boneName=dyn_left_strap_A}
    componentAppliedLinearAccClamp=<1500 1500 1500>
    componentLinearAccScale=<0.5 0.5 0.5>
    componentLinearVelScale=<0.6 0.6 0.6>
    physicsBodyDefinitions [{
      collisionType=CustomSphere
      sphereCollisionRadius=2.5
      boundBone {boneName=dyn_left_strap_A}
      boxExtents=<10 10 10>
      constraintSetup {angularLimitsMax=<0 60 15>, angularLimitsMin=<0 -25 -15>, angularTarget=<0 0 0>}
      localJointOffset=<2 0 0>
     }]
    planarLimits [{drivingBone {boneName=left_knife}, planeTransform {location=<0 0 0>, rotation=<rot 0 0 25>}}]
   }
   n2 = Anim Dynamics - Bone: dyn_left_strap_B!(ComponentPose=@n1, Alpha=1, GravityOverride:Vector="0,0,-980", ExternalForce="0,0,0") -> (Pose) {
    angularBiasOverride=1
    angularDampingOverride=1
    angularSpringConstant=3000
    bAngularSpring=true
    bOverrideAngularBias=true
    bOverrideAngularDamping=true
    bUseGravityOverride=true
    simulationSpace=RootRelative
    alphaBoolBlend {blendOption=Linear}
    boundBone {boneName=dyn_left_strap_B}
    componentAppliedLinearAccClamp=<1500 1500 1500>
    componentLinearAccScale=<0.5 0.5 0.5>
    componentLinearVelScale=<0.6 0.6 0.6>
    physicsBodyDefinitions [{
      collisionType=CustomSphere
      sphereCollisionRadius=3
      boundBone {boneName=dyn_left_strap_B}
      boxExtents=<10 10 10>
      constraintSetup {angularLimitsMax=<0 30 15>, angularLimitsMin=<0 -25 -15>, angularTarget=<0 0 0>}
      localJointOffset=<2 0 0>
     }]
    planarLimits [{drivingBone {boneName=left_knife}, planeTransform {location=<0 0 0>, rotation=<rot 0 0 25>}}]
   }
   n3 = Anim Dynamics - Bone: dyn_left_strap_C!(ComponentPose=@n2, Alpha=1, GravityOverride:Vector="0,0,-980", ExternalForce="0,0,0") -> (Pose) {
    angularBiasOverride=1
    angularDampingOverride=1
    angularSpringConstant=3000
    bAngularSpring=true
    bOverrideAngularBias=true
    bOverrideAngularDamping=true
    bUseGravityOverride=true
    simulationSpace=RootRelative
    alphaBoolBlend {blendOption=Linear}
    boundBone {boneName=dyn_left_strap_C}
    componentAppliedLinearAccClamp=<1500 1500 1500>
    componentLinearAccScale=<0.5 0.5 0.5>
    componentLinearVelScale=<0.6 0.6 0.6>
    physicsBodyDefinitions [{
      collisionType=CustomSphere
      sphereCollisionRadius=3
      boundBone {boneName=dyn_left_strap_C}
      boxExtents=<10 10 10>
      constraintSetup {angularLimitsMax=<0 30 -15>, angularLimitsMin=<0 -30 -15>, angularTarget=<0 0 0>}
      localJointOffset=<2 0 0>
     }]
    planarLimits [{drivingBone {boneName=left_knife}, planeTransform {location=<-0.0335 -2.2299 -0.006>, rotation=<rot 0 0 25>}}]
   }
   n4 = Anim Dynamics - Bone: dyn_right_strap_A!(ComponentPose=@n3, Alpha=1, GravityOverride:Vector="0,0,-980", ExternalForce="0,0,0") -> (Pose) {
    angularBiasOverride=1
    angularDampingOverride=1
    angularSpringConstant=3000
    bAngularSpring=true
    bOverrideAngularBias=true
    bOverrideAngularDamping=true
    bUseGravityOverride=true
    simulationSpace=RootRelative
    alphaBoolBlend {blendOption=Linear}
    boundBone {boneName=dyn_right_strap_A}
    componentAppliedLinearAccClamp=<1500 1500 1500>
    componentLinearAccScale=<0.5 0.5 0.5>
    componentLinearVelScale=<0.6 0.6 0.6>
    physicsBodyDefinitions [{
      collisionType=CustomSphere
      sphereCollisionRadius=2.5
      boundBone {boneName=dyn_right_strap_A}
      boxExtents=<10 10 10>
      constraintSetup {angularLimitsMax=<0 60 15>, angularLimitsMin=<0 -25 -15>, angularTarget=<0 0 0>}
      localJointOffset=<2 0 0>
     }]
    planarLimits [{drivingBone {boneName=right_knife}, planeTransform {location=<0 1 -1>, rotation=<rot 0 0 -25>}}]
   }
   n5 = Anim Dynamics - Bone: dyn_right_strap_B!(ComponentPose=@n4, Alpha=1, GravityOverride:Vector="0,0,-980", ExternalForce="0,0,0") -> (Pose) {
    angularBiasOverride=1
    angularDampingOverride=1
    angularSpringConstant=3000
    bAngularSpring=true
    bOverrideAngularBias=true
    bOverrideAngularDamping=true
    bUseGravityOverride=true
    simulationSpace=RootRelative
    alphaBoolBlend {blendOption=Linear}
    boundBone {boneName=dyn_right_strap_B}
    componentAppliedLinearAccClamp=<1500 1500 1500>
    componentLinearAccScale=<0.5 0.5 0.5>
    componentLinearVelScale=<0.6 0.6 0.6>
    physicsBodyDefinitions [{
      collisionType=CustomSphere
      sphereCollisionRadius=3
      boundBone {boneName=dyn_right_strap_B}
      boxExtents=<10 10 10>
      constraintSetup {angularLimitsMax=<0 30 30>, angularLimitsMin=<0 -25 -15>, angularTarget=<0 0 0>}
      localJointOffset=<2 0 0>
     }]
    planarLimits [{drivingBone {boneName=right_knife}, planeTransform {location=<0 0 -2>, rotation=<rot 0 0 -25>}}]
   }
   n6 = Anim Dynamics - Bone: dyn_right_strap_C!(ComponentPose=@n5, Alpha=1, GravityOverride:Vector="0,0,-980", ExternalForce="0,0,0") -> (Pose) {
    angularBiasOverride=1
    angularDampingOverride=1
    angularSpringConstant=3000
    bAngularSpring=true
    bOverrideAngularBias=true
    bOverrideAngularDamping=true
    bUseGravityOverride=true
    simulationSpace=RootRelative
    alphaBoolBlend {blendOption=Linear}
    boundBone {boneName=dyn_right_strap_C}
    componentAppliedLinearAccClamp=<1500 1500 1500>
    componentLinearAccScale=<0.5 0.5 0.5>
    componentLinearVelScale=<0.6 0.6 0.6>
    physicsBodyDefinitions [{
      collisionType=CustomSphere
      sphereCollisionRadius=3
      boundBone {boneName=dyn_right_strap_C}
      boxExtents=<10 10 10>
      constraintSetup {angularLimitsMax=<0 30 15>, angularLimitsMin=<0 -30 -15>, angularTarget=<0 0 0>}
      localJointOffset=<2 0 0>
     }]
    planarLimits [{drivingBone {boneName=right_knife}, planeTransform {location=<0 0 -2>, rotation=<rot 0 0 -25>}}]
   }
   n7 = RigidBody!(ComponentPose=Local To Component(LocalPose=Use cached pose 'INPUT POSE'()), Alpha, OverrideWorldGravity:Vector="0,0,0", ExternalForce) -> (Pose) {
    bOverrideWorldGravity=true
    overridePhysicsAsset=&/M_MED_PowerfulDozen_Pack_Physics
    simulationSpace=BaseBoneSpace
    alphaBoolBlend {blendOption=Linear}
    baseBoneRef {boneName=root}
    componentLinearAccScale=<0.4 0.4 0.4>
    componentLinearVelScale=<0.5 0.5 0.5>
   }
   Output Pose(Result=Blend Poses by bool(BlendPose_0=Component To Local(ComponentPose=@n7), BlendPose_1=Component To Local(ComponentPose=@n6), bActiveValue=Get bRigidBodyEnabled()) {blendType=Linear}) {layerGroup=None}
   n8 = "Transform (Modify) Bone - Bone: pack"!(
    ComponentPose=Local To Component(LocalPose=Input Pose())
    bAlphaBoolEnabled:bool=NOT Boolean(A=Get TMTN() -> (TMTN:bool))
    Translation=Get backpack_translation()
    Rotation=Get backpack_rotate()
    Scale
   ) -> (Pose) {
    alphaInputType=Bool
    rotationMode=BMM_Additive
    scaleMode=BMM_Additive
    scaleSpace=BCS_BoneSpace
    translationMode=BMM_Additive
    alphaBoolBlend {blendInTime=0.2, blendOption=Linear, blendOutTime=0.8}
    boneToModify {boneName=pack}
   }
   Save cached pose 'INPUT POSE'(Pose=Component To Local(ComponentPose=@n8)) {cacheName=INPUT POSE}
  ]}
 EventGraph {nodes [
   Event BlueprintUpdateAnimation()
   > n1 = Sequence!(execute) -> (then_0, then_1, then_2:Exec)
   n2 = Get Dynamic Material Instance(execute=[@n17.then, @n1.then_0, @n20.then_0], SkeletalMeshComponent=GetPartSkeletalMeshComponent())
   n3 = PickaxeItemDefinitionHasAnyMetaTag(Tags=Make Array("[0]"=(TagName="Cosmetics.MetaTags.PowerfulDozen")), bRequiresEquipped=true)
   > n4 = Branch(Condition=@n3)
   > n5 = Branch(Condition=Get bIsSkyDiving())
   SetScalarParameterValue(execute=[@n18.then, @n16.else, @n21.else, @n5.then, @n4.else], self=@n2, ParameterName=Hide Element 01, Value=0)
   SetScalarParameterValue(execute=[@n16.then, @n21.then, @n5.else], self=@n2, ParameterName=Hide Element 01, Value=1)
   n6 = Gravity Override State Function DEPRECATED(
    execute=@n1.then_1
    CharacterPartAnimInstance=Self-Reference()
    GameTimeInSeconds=GetGameTimeInSeconds()
    GravityOverrideParams=Get Gravity Override Params()
   )
   > Set Gravity Override(Gravity Override=@n6)
   n7 = GetDisplayName(Object=Get BodyMesh() -> (BodyMesh:Skeletal Mesh Component Object Reference))
   n8 = GetDisplayName(Object=Get BodyMesh())
   n9 = GetDisplayName(Object=Get BodyMesh())
   > Set TMTN(TMTN:bool=OR Boolean!(
     A=Contains(SearchIn=@n8, Substring=PowerFulDozen)
     B=Contains(SearchIn=@n9, Substring=QuietPeanuts)
     C:bool=Contains(SearchIn=@n7, Substring=QualityCreek)
    ) -> (ReturnValue)) -> (Output_Get:bool)
   n10 = GetOwnerSkeletalMeshComponent()
   n11 = BreakRotator(InRot=BreakTransform(InTransform=GetDeltaTransformFromRefPose(self=@n10, BoneName=spine_05, BaseName=spine_01)).Rotation)
   n12 = MapRangeClamped(Value=@n11.Pitch, InRangeA=30, InRangeB=-30, OutRangeA=-30, OutRangeB=30)
   n13 = MapRangeClamped(Value=@n11.Pitch, InRangeA=0, InRangeB=-30, OutRangeA=-5, OutRangeB=25)
   n14 = OR Boolean!(A=Get bIsSkydiveFloating(), B=Get bIsSkyDiving()) -> (ReturnValue)
   Set backpack_rotate(
    execute=@n1.then_2
    backpack_rotate=MakeRotator(Roll=MapRangeClamped(Value=@n11.Yaw, InRangeA=0, InRangeB=-30, OutRangeA=5, OutRangeB=25), Pitch=SelectFloat(A=@n13, B=@n12, bPickA=@n14))
   )
   > Set backpack_translation(backpack_translation=MakeVector(Y=MapRangeClamped(Value=@n11.Yaw, InRangeA=0, InRangeB=-30, OutRangeB=-6.5)))
   Event OnBeginPlayingEmote() {is=title K2Node_Event}
   > n15 = Set Emote Check(Emote Check:bool=true) -> (Output_Get:bool)
   > n16 = Branch(Condition=@n15)
   n17 = Event OnWeaponChanged()
   n18 = Event OnBeginSkydiving()
   n19 = Event OnCharacterCustomizationLoaded()
   > n20 = Sequence!(execute) -> (then_0, then_1, then_2, then_3:Exec)
   Set Skel Mesh(execute=@n20.then_1, Skel Mesh=@n19.CustomBaseMesh)
   Set BodyMesh(execute=@n20.then_2, BodyMesh=Get Array Item with EFort Custom Part Type(Array=@n19.CustomMeshParts))
   Event OnEndSkydiving()
   > n21 = Branch(Condition=@n3)
   "To String (Boolean)"(InBool=@n3)
  ]}
]
variables [
 backpack_rotate {type=Rotator}
 backpack_translation {type=Vector}
 BodyMesh {type=SkeletalMeshComponent}
 Emote Check {type=boolean}
 Gravity Override {type=Vector}
 Gravity Override Params {type=GravityOverrideParamsStruct}
 Skel Mesh {type=SkeletalMeshComponent}
 TMTN {type=boolean}
 TMTN Skydiving {type=boolean}
]
